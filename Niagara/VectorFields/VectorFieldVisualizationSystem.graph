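# Niagara system that visualizes a vector field supplied, with its transform and falloff, through User parameters.
# A burst arrow emitter distributes meshes across the field, orienting and scaling each by the sampled vector.
# A second mesh emitter builds the field's bounding box from twelve particles.
# A GPU sprite emitter continuously spawns particles in a sphere, sets velocity directly from the sampled field and colors them through a curve.
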NiagaraSystem /Niagara/VectorFields/VectorFieldVisualizationSystem
dependencies {usedDataInterfaces [&NiagaraDataInterfaceColorCurve], usedDynamicInputs [], usedModules [&SystemState]}
emitters {
 VectorFieldArrowEmitter {
  emitterUpdateScript [
   &EmitterState()
   &SpawnBurst_Instantaneous(Spawn Count=1000, Use Spawn Probability=false)
  ]
  particleUpdateScript [
   &/DistributeArrowsForVectorField(
    FieldCoordinates:ENiagaraCoordinateSpace=$User.FieldCoordinates
    FieldRotate:Quat4f=$User.FieldRotation
    FieldScale:Vector3f=$User.FieldScale
    FieldTranslate:Vector3f=$User.FieldLocation
    VectorField:NiagaraDataInterfaceVectorField=$User.VectorField
   )
   &/SampleVectorField(
    ApplyFalloff:NiagaraBool=$User.FieldApplyFalloff
    FalloffDistance:NiagaraFloat=$User.FieldFalloffDistance
    FieldCoordinates:ENiagaraCoordinateSpace=$User.FieldCoordinates
    FieldRotate:Quat4f=$User.FieldRotation
    FieldScale:Vector3f=$User.FieldScale
    FieldTranslate:Vector3f=$User.FieldLocation
    UseExponentialFalloff:NiagaraBool=$User.FieldUseExponentialFalloff
    VectorField:NiagaraDataInterfaceVectorField=$User.VectorField
   )
   &/ApplyVectorField()
   /SetVariables(
    Particles.MeshOrientation=&FindQuatBetweenTwoVectors(From Vector=<1 0 0>, To Vector=$Transient.SampleVectorFieldSampledVector)
    Particles.Scale=&VectorFromFloat(Value=HlslExpression() {hlslExpression="0.1/256.0*length(Output.SampleVectorField.SampledVector)*Particles.Scale"})
   ) {isSetParameters=true, module=SetVariables_#1}
  ]
  rendererClasses [&NiagaraMeshRendererProperties]
  renderers [{rendererClass=&NiagaraMeshRendererProperties, rendererIndex=0}]
 }
 VectorFieldBoundingBoxEmitter {
  emitterUpdateScript [
   &EmitterState()
   &SpawnBurst_Instantaneous(Spawn Count=12, Use Spawn Probability=false)
  ]
  particleUpdateScript [&/ConstructBoundingBoxForVectorField(
    FieldCoordinates:ENiagaraCoordinateSpace=$User.FieldCoordinates
    Rotation:Quat4f=$User.FieldRotation
    Scale:Vector3f=$User.FieldScale
    Translate:Vector3f=$User.FieldLocation
    VectorField:NiagaraDataInterfaceVectorField=$User.VectorField
   ) {module=ConstructBoundingBoxForVectorField001}]
  rendererClasses [&NiagaraMeshRendererProperties]
  renderers [{rendererClass=&NiagaraMeshRendererProperties, rendererIndex=0}]
 }
 VectorFieldParticleEmitter {
  simTarget=GPUComputeSim
  emitterUpdateScript [
   &EmitterState(Life Cycle Mode=::NewEnumerator1, UseLoopDelay=false)
   &SpawnRate(SpawnRate=1250, Use Spawn Probability=false)
  ]
  particleSpawnScript [
   /Niagara/Modules/Spawn/Location/SphereLocation(MaskSpawnGroup=false, Module.Override Randomness=false, Sphere Radius=150, Surface Only=false)
   /SetVariables(
    Particles.Lifetime=&UniformRangedFloat(Minimum=0.5, Maximum=1)
    Particles.SpriteSize=&Vector2DFromFloat(Value=15)
   ) {isSetParameters=true, module=SetVariables_#1}
  ]
  particleUpdateScript [
   &SampleVectorField(
    ApplyFalloff:NiagaraBool=$User.FieldApplyFalloff
    FalloffDistance:NiagaraFloat=$User.FieldFalloffDistance
    FieldCoordinates:ENiagaraCoordinateSpace=$User.FieldCoordinates
    FieldRotate:Quat4f=$User.FieldRotation
    FieldScale:Vector3f=$User.FieldScale
    FieldTranslate:Vector3f=$User.FieldLocation
    UseExponentialFalloff:NiagaraBool=$User.FieldUseExponentialFalloff
    VectorField:NiagaraDataInterfaceVectorField=$User.VectorField
   )
   &ApplyVectorField()
   /SetVariables(Particles.Velocity=$Transient.SampleVectorFieldSampledVector) {isSetParameters=true}
   &ParticleState() {module=UpdateAge}
   &Color(Color=&ColorFromCurve(ColorCurve=DataInterface() {propertyValues {
       AlphaCurve {keys [<at 0 1 RCIM_Cubic>,<at 0.9977 1 RCIM_Cubic>]}
       BlueCurve {keys [{arriveTangent=-1.0005, interpMode=RCIM_Cubic, time=0, value=1}, <at 0.9996 0>]}
       GreenCurve {keys [<at 0 0 RCIM_Cubic>,<at 0.9996 1>]}
       RedCurve {keys [{arriveTangent=0.189, interpMode=RCIM_Cubic, time=0, value=0}, <at 0.9996 0.2661>]}
      }}))
   &SolveForcesAndVelocity(Clamp Velocity=false, Limit Acceleration=false)
  ]
  rendererClasses [&NiagaraSpriteRendererProperties]
  renderers [{rendererClass=&NiagaraSpriteRendererProperties}]
 }
}
summary {
 systemName=VectorFieldVisualizationSystem
 emitters [
  {emitterName=VectorFieldBoundingBoxEmitter, rendererClasses [&NiagaraMeshRendererProperties]}
  {emitterName=VectorFieldArrowEmitter, rendererClasses [&NiagaraMeshRendererProperties]}
  {emitterName=VectorFieldParticleEmitter, simTarget=GPUComputeSim, rendererClasses [&NiagaraSpriteRendererProperties]}
 ]
 userVariables [
  User.FieldApplyFalloff {type=NiagaraBool, defaultValue=true}
  User.FieldCoordinates {
   type=ENiagaraCoordinateSpace
   description=""
   defaultValue {
    struct=NiagaraExt_VariableValue_Enum
    value {displayName=Simulation, enum=/Script/Niagara.ENiagaraCoordinateSpace, enumName=ENiagaraCoordinateSpace::Simulation}
   }
  }
  User.FieldFalloffDistance {type=NiagaraFloat, defaultValue=45, description=""}
  User.FieldIntensity {type=NiagaraFloat, defaultValue=1, description=""}
  User.FieldLocation {type=Vector3f, description="", defaultValue {struct=Vector3f, value=<0 0 0>}}
  User.FieldRotation {type=Quat4f, description="", defaultValue {struct=Quat4f, value=<0 0 0 1>}}
  User.FieldScale {type=Vector3f, description="", defaultValue {struct=Vector3f, value=<1 1 1>}}
  User.FieldUseExponentialFalloff {type=NiagaraBool, defaultValue=true}
  User.VectorField {
   type=NiagaraDataInterfaceVectorField
   description=""
   defaultValue {
    struct=NiagaraExt_VariableValue_DataInterface
    value {
     dataInterface=&self:NiagaraDataInterfaceVectorField_0
     dataInterfaceClass=/Script/Niagara.NiagaraDataInterfaceVectorField
    }
   }
  }
 ]
}
